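; Looping gameplay cue for the Caretaker's melee charging, which drives the charge-up visual effect.
; On looping start it casts the first spawned particle component to Niagara and passes the cue's WarmUpTime variable into the effect's parameter of the same name.
; It then waits that same duration and deactivates the component, so the effect stops emitting when the warm-up ends.
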(event EventBeginPlay)

(event Collision|EventActorBeginOverlap
  (params (OtherActor :type Actor%20Object%20Reference)))

(event EventOnLoopingStart(EitherVFX-preferthisversion)
  (params (MyTarget :type Actor%20Object%20Reference) (Parameters :type Gameplay%20Cue%20Parameters%20Structure%20%28by%20ref%29) (ParticleComponents :type Array%20of%20FXSystem%20Component%20Object%20References) (AudioComponents :type Array%20of%20Audio%20Component%20Object%20References))
  (bind _warmuptime (Variables|Default|GetWarmUpTime))
  (bind _asniagara_particle_system_component (Utilities|Casting|CastToNiagaraComponent (Utilities|Array|Get(acopy) ParticleComponents)))
  (Effects|Components|ParticleSystem|SetFloatParameter _asniagara_particle_system_component "WarmUpTime" _warmuptime)
  (Utilities|FlowControl|Delay 0 _warmuptime)
  (Components|Activation|Deactivate _asniagara_particle_system_component))

(fn ConstructionScript)
