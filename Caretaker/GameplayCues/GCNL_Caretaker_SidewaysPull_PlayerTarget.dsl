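; Looping gameplay cue for the player targeted by the Caretaker's Sideways pull, handling its particles and audio.
; On loop start it stores the target and feeds the Caretaker's skeletal mesh into the Niagara component.
; The target entering the Sideways world triggers cleanup, with a timed delay as fallback.
; A DoOnce-guarded cleanup unbinds the state delegate, deactivates particles, fades looping audio and plays an attached detach sound.
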(event EventBeginPlay)

(event Collision|EventActorBeginOverlap
  (params (OtherActor :type Actor%20Object%20Reference)))

(event EventOnLoopingStart(EitherVFX-preferthisversion)
  (params (MyTarget :type Actor%20Object%20Reference) (Parameters :type Gameplay%20Cue%20Parameters%20Structure%20%28by%20ref%29) (ParticleComponents :type Array%20of%20FXSystem%20Component%20Object%20References) (AudioComponents :type Array%20of%20Audio%20Component%20Object%20References))
  (bind (_normalizedmagnitude _rawmagnitude _effectcontext) (Ability|GameplayCue|BreakGameplayCueParameters Parameters))
  (bind _playertarget (Variables|Default|GetPlayerTarget))
  (Variables|Default|SetPlayerTarget MyTarget)
  (Utilities|FlowControl|Sequence
    (:then_0
      (CallFunction|OpenDeactivateGate)
      (bind _asniagara_particle_system_component (Utilities|Casting|CastToNiagaraComponent (Utilities|Array|Get(acopy) ParticleComponents)))
      (Variables|Default|SetNiagaraParticleSystem_Item1_Ref _asniagara_particle_system_component)
      (bind _asnpc_pawn_irwin_monster_caretaker (Utilities|Casting|CastToNPC_Pawn_Irwin_Monster_Caretaker (Ability|EffectContext|GetEffectCauser _effectcontext)))
      (Niagara|SetNiagaraSkeletalMeshComponent (Variables|Default|GetNiagaraParticleSystem_Item1_Ref) "CaretakerMesh" (Variables|/Script/Engine.Character|GetMesh _asnpc_pawn_irwin_monster_caretaker)))
    (:then_1
      ;[1 1 1] Track if in sideways to deactivate particles more accurately (timer is fallback)
      (Utilities|IsValid _playertarget
        (:Is%20Valid
          (bind _returnvalue (Sideways|GetSidewaysStateComponent _playertarget))
          (Utilities|IsValid _returnvalue
            (:Is%20Valid
              (EventDispatchers|AssignOnSidewaysStateChangedDelegate _returnvalue (delegate Custom|OnSidewaysStateChangedDelegate_Event_1)))))))
    (:then_2
      (Utilities|FlowControl|Delay 0 2.0)
      (CallFunction|DeactivateEffectstoPlayer))))

(event Custom|OnSidewaysStateChangedDelegate_Event_1
  (params (SidewaysStateComponent :type Sideways%20State%20Component%20Object%20Reference) (SidewaysPlayspace :type Sideways%20Playspace%20Object%20Reference) (OldState :type ESidewaysState%20Enum :default "InNormalWorld") (NewState :type ESidewaysState%20Enum :default "InNormalWorld"))
  ;[1 1 1] Track if in sideways to deactivate particles more accurately (timer is fallback)
  (if (Utilities|Enum|NotEqual(Enum) OldState NewState)
    (switch Utilities|FlowControl|Switch|SwitchonESidewaysState NewState
      (:InSidewaysWorld
        (CallFunction|DeactivateEffectstoPlayer)))))

(event Custom|DeactivateEffectsToPlayer
  (bind _playertarget (Variables|Default|GetPlayerTarget))
  ;[1 1 1] Deactivate section
  (Utilities|FlowControl|DoOnce false
    (:Completed
      (Utilities|IsValid _playertarget
        (:Is%20Valid
          (bind _returnvalue (Sideways|GetSidewaysStateComponent _playertarget))
          (Utilities|IsValid _returnvalue
            (:Is%20Valid
              (EventDispatchers|UnbindEventfromOnSidewaysStateChangedDelegate _returnvalue (delegate Custom|OnSidewaysStateChangedDelegate_Event_1)))
            (:Is%20Not%20Valid
              _))
          (Components|Activation|Deactivate (Variables|Default|GetNiagaraParticleSystem_Item1_Ref))
          (if (Audio|Components|Audio|IsPlaying (Utilities|Array|Get(acopy) (stale EventOnLoopingStart(EitherVFX-preferthisversion) AudioComponents)))
            (Audio|Components|Audio|FadeOut (Utilities|Array|Get(acopy) (stale EventOnLoopingStart(EitherVFX-preferthisversion) AudioComponents)) 0.15 0.0 "Linear" :!class /Script/Engine.AudioComponent)
            (else
              _))
          (bind _asfort_pawn (Utilities|Casting|CastToFortPawn (Variables|Default|GetPlayerTarget)))
          (Audio|SpawnSoundAttached "/Caretaker/Sounds/CareTaker_Sideways_Pull_Player_Dettach_Cue.CareTaker_Sideways_Pull_Player_Dettach_Cue" (Variables|/Script/Engine.Character|GetMesh _asfort_pawn)))
        (:Is%20Not%20Valid
          _)))))

(event Custom|OpenDeactivateGate
  (bind _playertarget (Variables|Default|GetPlayerTarget))
  ;[1 1 1] Deactivate section
  (Utilities|FlowControl|DoOnce false
    (:Completed
      (Utilities|IsValid _playertarget
        (:Is%20Valid
          (bind _returnvalue (Sideways|GetSidewaysStateComponent _playertarget))
          (Utilities|IsValid _returnvalue
            (:Is%20Valid
              (EventDispatchers|UnbindEventfromOnSidewaysStateChangedDelegate _returnvalue (delegate Custom|OnSidewaysStateChangedDelegate_Event_1)))
            (:Is%20Not%20Valid
              _))
          (Components|Activation|Deactivate (Variables|Default|GetNiagaraParticleSystem_Item1_Ref))
          (if (Audio|Components|Audio|IsPlaying (Utilities|Array|Get(acopy) (stale EventOnLoopingStart(EitherVFX-preferthisversion) AudioComponents)))
            (Audio|Components|Audio|FadeOut (Utilities|Array|Get(acopy) (stale EventOnLoopingStart(EitherVFX-preferthisversion) AudioComponents)) 0.15 0.0 "Linear" :!class /Script/Engine.AudioComponent)
            (else
              _))
          (bind _asfort_pawn (Utilities|Casting|CastToFortPawn (Variables|Default|GetPlayerTarget)))
          (Audio|SpawnSoundAttached "/Caretaker/Sounds/CareTaker_Sideways_Pull_Player_Dettach_Cue.CareTaker_Sideways_Pull_Player_Dettach_Cue" (Variables|/Script/Engine.Character|GetMesh _asfort_pawn)))
        (:Is%20Not%20Valid
          _)))))

(fn ConstructionScript)
