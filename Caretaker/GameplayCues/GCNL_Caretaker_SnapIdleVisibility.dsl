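; Looping gameplay cue that shows the Caretaker monster's idle eyes Niagara effect while the cue is active.
; On looping start it casts the target to the Caretaker pawn and, only if the pawn is not in Sideways, attaches the effect to its mesh and stores the component.
; It sets the effect's Sideways float parameter to 1 after spawning and again on removal, when it deactivates the stored component if valid.
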(event EventOnLoopingStart(EitherVFX-preferthisversion)
  (params (MyTarget :type Actor%20Object%20Reference) (Parameters :type Gameplay%20Cue%20Parameters%20Structure%20%28by%20ref%29) (ParticleComponents :type Array%20of%20FXSystem%20Component%20Object%20References) (AudioComponents :type Array%20of%20Audio%20Component%20Object%20References))
  (bind _asnpc_pawn_irwin_monster_caretaker (Utilities|Casting|CastToNPC_Pawn_Irwin_Monster_Caretaker MyTarget))
  (if (not (Variables|/Caretaker/Pawns/NPC_Pawn_Irwin_Monster_Caretaker.NPC_Pawn_Irwin_Monster_Caretaker_C|GetisInSideways _asnpc_pawn_irwin_monster_caretaker))
    (bind _returnvalue (Niagara|SpawnSystemAttached "/Caretaker/VFX/NS_Caretaker_Idle_Eyes.NS_Caretaker_Idle_Eyes" (Variables|/Script/Engine.Character|GetMesh _asnpc_pawn_irwin_monster_caretaker) "None" "0, 0, 0" "0, 0, 0" "KeepRelativeOffset" true))
    (bind _output_get (Variables|Default|SetSpawned%20Idle _returnvalue))
    (Effects|Components|ParticleSystem|SetFloatParameter _output_get "Sideways" 1.0)))

(event EventOnRemoval(EitherVFX-preferthisversion)
  (params (MyTarget :type Actor%20Object%20Reference) (Parameters :type Gameplay%20Cue%20Parameters%20Structure%20%28by%20ref%29) (ParticleComponents :type Array%20of%20FXSystem%20Component%20Object%20References) (AudioComponents :type Array%20of%20Audio%20Component%20Object%20References) (BurstCameraShakeInstance :type Camera%20Shake%20Base%20Object%20Reference) (BurstDecalInstance :type Decal%20Actor%20Object%20Reference))
  (bind _spawned_idle (Variables|Default|GetSpawned%20Idle))
  (Utilities|IsValid _spawned_idle
    (:Is%20Valid
      (Effects|Components|ParticleSystem|SetFloatParameter _spawned_idle "Sideways" 1.0)
      (Components|Activation|Deactivate _spawned_idle))))

(fn ConstructionScript)
