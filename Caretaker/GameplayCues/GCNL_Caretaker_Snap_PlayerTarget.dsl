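; Looping gameplay cue for the Caretaker's Snap on a player target, driving tether VFX and audio.
; Loop start links the Niagara effect to the Caretaker mesh, sets tighten and break distances, starts 1P or 3P loop audio and binds MeleeWarningAttackFired.
; That event tightens the effect, spawns a pull hit, then after a brief delay deactivates it and swaps to end audio.
; Removal unbinds and, if no pull occurred, fades the loop.
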(event EventBeginPlay)

(event Collision|EventActorBeginOverlap
  (params (OtherActor :type Actor%20Object%20Reference)))

(event EventOnLoopingStart(EitherVFX-preferthisversion)
  (params (MyTarget :type Actor%20Object%20Reference) (Parameters :type Gameplay%20Cue%20Parameters%20Structure%20%28by%20ref%29) (ParticleComponents :type Array%20of%20FXSystem%20Component%20Object%20References) (AudioComponents :type Array%20of%20Audio%20Component%20Object%20References))
  (bind (_normalizedmagnitude _rawmagnitude _effectcontext) (Ability|GameplayCue|BreakGameplayCueParameters Parameters))
  (bind _niagaraparticlesystem_item1_ref (Variables|Default|GetNiagaraParticleSystem_Item1_Ref))
  (Variables|Default|SetisPulled false)
  (bind _asfort_player_pawn (Utilities|Casting|CastToFortPlayerPawn MyTarget)
    (:then
      (Variables|Default|SetTarget%20Player _asfort_player_pawn)
      (bind _asniagara_particle_system_component (Utilities|Casting|CastToNiagaraComponent (Utilities|Array|Get(acopy) ParticleComponents)))
      (Variables|Default|SetNiagaraParticleSystem_Item1_Ref _asniagara_particle_system_component)
      (bind _asnpc_pawn_irwin_monster_caretaker (Utilities|Casting|CastToNPC_Pawn_Irwin_Monster_Caretaker (Ability|EffectContext|GetEffectCauser _effectcontext)))
      (Variables|Default|SetCaretakerRef _asnpc_pawn_irwin_monster_caretaker)
      (Niagara|SetNiagaraSkeletalMeshComponent _niagaraparticlesystem_item1_ref "CaretakerMesh" (Variables|/Script/Engine.Character|GetMesh _asnpc_pawn_irwin_monster_caretaker))
      (Effects|Components|ParticleSystem|SetFloatParameter _niagaraparticlesystem_item1_ref "TightDistance" (Variables|Default|GetDistance%20to%20Tight))
      (Effects|Components|ParticleSystem|SetFloatParameter _niagaraparticlesystem_item1_ref "BreakDistance" (Variables|Default|GetDistance%20to%20Break))
      (Default|AssignMeleeWarningAttackFired _asnpc_pawn_irwin_monster_caretaker (delegate Custom|MeleeWarningAttackFired_Event_0))
      (CallFunction|PlaySnapAudioonPlayer))
    (:CastFailed
      (Development|PrintString 0 "Caretaker's Snap Target is not a Player Pawn"))))

(event Custom|MeleeWarningAttackFired_Event_0
  (Effects|Components|ParticleSystem|SetBoolParameter (Variables|Default|GetNiagaraParticleSystem_Item1_Ref) "Tight" true)
  (Variables|Default|SetisPulled true)
  (Niagara|SpawnSystemAttached "/Caretaker/FX/Snap/NS_Caretaker_Snap_PullHit.NS_Caretaker_Snap_PullHit" (Variables|/Script/Engine.Character|GetMesh (Variables|Default|GetTarget%20Player)) "spine_04" "0, 0, 0" "0, 0, 0" "KeepRelativeOffset" true)
  (Utilities|FlowControl|Delay 0 0.125)
  (Components|Activation|Deactivate (Variables|Default|GetNiagaraParticleSystem_Item1_Ref))
  (bind _snapaudio (Variables|Default|GetSnapAudio)
    (:then
      (Audio|Components|Audio|FadeOut _snapaudio 0.25 0.0 "Linear" :!class /Script/Engine.AudioComponent)
      (CallFunction|PlaySnapEndAudioonPlayer))
    (:else)))

(event EventOnRemoval(EitherVFX-preferthisversion)
  (params (MyTarget :type Actor%20Object%20Reference) (Parameters :type Gameplay%20Cue%20Parameters%20Structure%20%28by%20ref%29) (ParticleComponents :type Array%20of%20FXSystem%20Component%20Object%20References) (AudioComponents :type Array%20of%20Audio%20Component%20Object%20References) (BurstCameraShakeInstance :type Camera%20Shake%20Base%20Object%20Reference) (BurstDecalInstance :type Decal%20Actor%20Object%20Reference))
  (bind _caretakerref (Variables|Default|GetCaretakerRef))
  (Utilities|IsValid _caretakerref
    (:Is%20Valid
      (Default|UnbindEventfromMeleeWarningAttackFired _caretakerref (delegate Custom|MeleeWarningAttackFired_Event_0)))
    (:Is%20Not%20Valid
      _))
  (if (not (Variables|Default|GetisPulled))
    (bind _snapaudio (Variables|Default|GetSnapAudio)
      (:then
        (Audio|Components|Audio|FadeOut _snapaudio 0.25 0.0 "Linear" :!class /Script/Engine.AudioComponent))
      (:else))))

(fn PlaySnapAudioOnPlayer
  (if (Pawn|IsLocallyViewed (Variables|Default|GetTarget%20Player))
    (bind _snapaudio (Variables|Default|GetSnapAudio)
      (:then
        (Audio|Components|Audio|FadeOut _snapaudio 0.15 0.0 "Linear" :!class /Script/Engine.AudioComponent))
      (:else
        _))
    (bind _returnvalue (Audio|SpawnSound2D 0 "/Caretaker/Sounds/CareTaker_Snap_Player_Loop_1p_Cue.CareTaker_Snap_Player_Loop_1p_Cue"))
    (Variables|Default|SetSnapAudio _returnvalue)
    (else
      (bind _snapaudio_2 (Variables|Default|GetSnapAudio)
        (:then
          (Audio|Components|Audio|FadeOut _snapaudio_2 0.15 0.0 "Linear" :!class /Script/Engine.AudioComponent))
        (:else
          _))
      (bind _returnvalue_2 (Audio|SpawnSoundAttached "/Caretaker/Sounds/CareTaker_Snap_Player_Loop_3p_Cue.CareTaker_Snap_Player_Loop_3p_Cue" (Variables|/Script/Engine.Character|GetMesh (Variables|Default|GetTarget%20Player))))
      (Variables|Default|SetSnapAudio _returnvalue_2))))

(fn PlaySnapEndAudioOnPlayer
  (if (Pawn|IsLocallyViewed (Variables|Default|GetTarget%20Player))
    (bind _snapendaudio (Variables|Default|GetSnapEndAudio)
      (:then
        (Audio|Components|Audio|FadeOut _snapendaudio 0.15 0.0 "Linear" :!class /Script/Engine.AudioComponent))
      (:else
        _))
    (bind _returnvalue (Audio|SpawnSound2D 0 "/Caretaker/Sounds/CareTaker_Snap_Player_Stop_1p_Cue.CareTaker_Snap_Player_Stop_1p_Cue"))
    (Variables|Default|SetSnapEndAudio _returnvalue)
    (else
      (bind _snapendaudio_2 (Variables|Default|GetSnapEndAudio)
        (:then
          (Audio|Components|Audio|FadeOut _snapendaudio_2 0.15 0.0 "Linear" :!class /Script/Engine.AudioComponent))
        (:else
          _))
      (bind _returnvalue_2 (Audio|SpawnSoundAttached "/Caretaker/Sounds/CareTaker_Snap_Playey_Stop_3p_Cue.CareTaker_Snap_Playey_Stop_3p_Cue" (Variables|/Script/Engine.Character|GetMesh (Variables|Default|GetTarget%20Player))))
      (Variables|Default|SetSnapEndAudio _returnvalue_2))))

(fn ConstructionScript)
