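; Burst gameplay cue for the Caretaker's melee impact, spawning effects at the FX_Melee_Impact socket on the target's skeletal mesh.
; It shows a gunshot-type sound indicator to neutral and hostile teams, then traces down onto the socket to detect water.
; A hit spawns water-impact Niagara fed surface location, normal, depth, body index and wave attenuation from the water body.
; Otherwise it spawns the standard impact Niagara and a decal actor.
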(fn OnBurst(EitherVFX-preferthisversion)
  (params (MyTarget :type Actor%20Object%20Reference) (Parameters :type Gameplay%20Cue%20Parameters%20Structure%20%28by%20ref%29) (ParticleComponents :type Array%20of%20FXSystem%20Component%20Object%20References) (AudioComponents :type Array%20of%20Audio%20Component%20Object%20References) (BurstCameraShakeInstance :type Camera%20Shake%20Base%20Object%20Reference) (BurstDecalInstance :type Decal%20Actor%20Object%20Reference))
  (bind _returnvalue (Actor|GetComponentbyClass MyTarget "/Script/Engine.SkeletalMeshComponent"))
  (bind _returnvalue_2 (Transformation|GetSocketLocation _returnvalue "FX_Melee_Impact"))
  (Utilities|IsValid MyTarget
    (:Is%20Valid
      (Utilities|IsValid _returnvalue
        (:Is%20Valid
          (Fortnite|ShowSoundIndicatorLocalforTeam :WorldContextObject MyTarget :TrackedActor MyTarget :Location (Transformation|GetActorLocation MyTarget) :MaxAudibleDistance 1.08e+04 :IndicatorType "Gunshot" :Team (Team|GetActorTeam 0) :Affiliations (Utilities|Array|MakeArray "Neutral" "Hostile") :Instigator MyTarget :Tint "(R=0.000000,G=0.000000,B=0.000000,A=0.000000)" :AttenuationMaxRangeMultiplier 1.0 :AttenuationIntensityMultiplier 1.0)
          (bind (_outhit _returnvalue_3) (Collision|LineTraceByChannel 0 (+ _returnvalue_2 "0, 0,750.000000") _returnvalue_2 "TraceTypeQuery12"))
          (if _returnvalue_3
            (bind _asfort_water_body_actor (Utilities|Casting|CastToFortWaterBodyActor (.HitActor (Collision|BreakHitResult _outhit))))
            (bind _outwatersurfaceinfo (WaterBody|GetWaterSurfaceInfoStruct _asfort_water_body_actor (.ImpactPoint (Collision|BreakHitResult _outhit)) true false false))
            (bind _returnvalue_4 (Niagara|SpawnSystematLocation 0 "/Caretaker/FX/Melee/NS_Caretaker_Melee_Impact_Water.NS_Caretaker_Melee_Impact_Water" (Utilities|Struct|BreakWaterSurfaceInfo _outwatersurfaceinfo)))
            (Effects|Components|ParticleSystem|SetVectorParameter _returnvalue_4 "Water Location" (Utilities|Struct|BreakWaterSurfaceInfo _outwatersurfaceinfo))
            (Effects|Components|ParticleSystem|SetVectorParameter _returnvalue_4 "Water Normal" (.WaterPlaneNormal (Utilities|Struct|BreakWaterSurfaceInfo _outwatersurfaceinfo)))
            (Effects|Components|ParticleSystem|SetFloatParameter _returnvalue_4 "WaterDepth" (.WaterDepth (Utilities|Struct|BreakWaterSurfaceInfo _outwatersurfaceinfo)))
            (Effects|Components|ParticleSystem|SetFloatParameter _returnvalue_4 "WaterBodyIndex" (Math|Conversions|ToFloat(Integer) (.WaterBodyIdx (Utilities|Struct|BreakWaterSurfaceInfo _outwatersurfaceinfo))))
            (Effects|Components|ParticleSystem|SetFloatParameter _returnvalue_4 "WaveAttenuationFactor" (Variables|/Script/Water.WaterBodyComponent|GetTargetWaveMaskDepth (Variables|/Script/Water.WaterBody|GetWaterBodyComponent _asfort_water_body_actor)))
            (else
              (Niagara|SpawnSystematLocation 0 "/Caretaker/FX/Melee/NS_Caretaker_Melee_Impact.NS_Caretaker_Melee_Impact" _returnvalue_2)
              (Game|SpawnActorfromClass 0 "/Caretaker/FX/Melee/B_Decal_Caretaker_Melee.B_Decal_Caretaker_Melee_C" (Math|Transform|MakeTransform _returnvalue_2) "Undefined" "MultiplyWithRoot" 0 0))))))))
